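# Sound cue for the Hightower Radish pet, named for its gun-ready reaction at the start of a DBNO (down but not out) state.
# A single wave player plays one sound wave, with no randomization or other processing nodes.
# It uses the close-range pet attenuation and the creature pets sound class, with a slight volume boost.
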SoundCue /Game/Athena/Sounds/Pets/Hightower_Radish/HightowerRadish_Pet_DBNOStart_Ready_Gun
root=WavePlayer_0
nodes [{id=WavePlayer_0, properties {soundWaveAssetPtr=&/Pet_HightowerRadish_DBNO_Start_Ready_Gun_01}}]
properties {
 attenuationSettings=&Fort_Pets_Close_Athena_Att
 maxDistance=1800.0001
 soundClassObject=&Fort_Creatures_Pets
 volumeMultiplier=1.12
}
