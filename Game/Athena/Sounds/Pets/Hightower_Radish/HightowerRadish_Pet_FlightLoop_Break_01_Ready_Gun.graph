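# Sound cue for the Hightower Radish pet, named as a flight-loop break "Ready Gun" variation.
# It randomly picks one of two equally weighted waves (takes 01 and 03), then applies a slight random downward pitch variation through a modulator.
# Output plays in the creature pets sound class at reduced volume, using close-range pet attenuation.
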SoundCue /Game/Athena/Sounds/Pets/Hightower_Radish/HightowerRadish_Pet_FlightLoop_Break_01_Ready_Gun
root=Modulator_0
nodes [
 {id=Modulator_0, children [Random_0], properties {pitchMax=1, pitchMin=0.9, volumeMax=1, volumeMin=1}}
 {id=Random_0, children [WavePlayer_1,WavePlayer_2], properties {weights [1,1]}}
 {
  id=WavePlayer_1
  properties {soundWaveAssetPtr=&/Pet_HightowerRadish_FlightLoop_Break01_Ready_Gun_01}
 }
 {
  id=WavePlayer_2
  properties {soundWaveAssetPtr=&/Pet_HightowerRadish_FlightLoop_Break01_Ready_Gun_03}
 }
]
properties {
 attenuationSettings=&Fort_Pets_Close_Athena_Att
 maxDistance=1800.0001
 soundClassObject=&Fort_Creatures_Pets
 volumeMultiplier=0.448
}
